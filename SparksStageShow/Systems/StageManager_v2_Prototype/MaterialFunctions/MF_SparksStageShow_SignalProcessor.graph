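# Material function for the Sparks Stage Show that turns a linear measure progression signal into a beat-pulsed signal clamped to 0–1.
# The measure is split by beats per measure, multiplier and offset, then wrapped with Frac.
# Eased rise and fall ramps, timed in beat fractions independent of the multiplier, meet at each beat and combine via Max.
# AnimationMode scales the fall for linear or ping-pong shapes before remapping between Min and Peak.
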MaterialFunction /SparksStageShow/Systems/StageManager_v2_Prototype/MaterialFunctions/MF_SparksStageShow_SignalProcessor
graph {
 properties {
  description=Receives a linear time signal and outputs a beating signal. All parameters are measure or beat scaled.
  userExposedCaption=Sparks Stage Show Signal Processor
 }
 nodes [
  n1 = FunctionInput() {
   bUsePreviewValueAsDefault=true
   description=Duration of the signal rise phase. Unit is 1/x beat. Example: setting this to 4 means the rise duration is 1/4 beat.
   inputName=Rise Time
   inputType=FunctionInput_Scalar
   sortPriority=6
   previewValue=<8 0 0 1>
  }
  n2 = FunctionInput() {
   bUsePreviewValueAsDefault=true
   description="Multiplies the number of beats in the measure while keeping the beat signal lenght unmodified. Useful to reduce the number of notes, ex: play just the first note in the measure, or skip every other note, etc."
   inputName=Beat Multiplier
   inputType=FunctionInput_Scalar
   sortPriority=2
   previewValue=<1 0 0 1>
  }
  n3 = Divide(A=@n1, B=@n2)
  n4 = FunctionInput(Preview=Multiply(A=Time()) {constB=0.25}) {
   bUsePreviewValueAsDefault=true
   description=The measure progression signal that will be used.
   inputName=Measure Signal
   inputType=FunctionInput_Scalar
  }
  n5 = FunctionInput() {
   bUsePreviewValueAsDefault=true
   description=How many beats the measure period will be divided
   inputName=Beats Per Measure
   inputType=FunctionInput_Scalar
   sortPriority=1
   previewValue=<4 0 0 1>
  }
  n6 = Multiply(A=@n5, B=@n2)
  n7 = FunctionInput() {
   bUsePreviewValueAsDefault=true
   description="Offsets the pattern using beats as unit. 1 = one beat offset"
   inputName=BeatOffset
   inputType=FunctionInput_Scalar
   sortPriority=3
  }
  n8 = Frac(None=Subtract(A=Multiply(A=@n4, B=@n6), B=Multiply(A=@n2, B=@n7)))
  n9 = Subtract(A=Multiply(A=@n8, B=@n3), B=@n3)
  n10 = FunctionInput() {
   bUsePreviewValueAsDefault=true
   description="Makes the rise phase not linear by applying a exponential curve. Values above 1 creates a easy out effect; below 1 creates a easy in effect."
   inputName=Rise Easing
   inputType=FunctionInput_Scalar
   sortPriority=7
   previewValue=<1 0 0 1>
  }
  n11 = Power(Base=Saturate(None=Add(A=@n9)), Exp=@n10) {constExponent=1}
  n12 = OneMinus(None=@n8)
  n13 = FunctionInput() {
   bUsePreviewValueAsDefault=true
   description=Duration of the signal fall phase. Unit is 1/x beat. Example: setting this to 4 means the fall duration is 1/4 beat.
   inputName=Fall Time
   inputType=FunctionInput_Scalar
   sortPriority=8
   previewValue=<4 0 0 1>
  }
  n14 = Divide(A=@n13, B=@n2)
  n15 = Add(A=Subtract(A=Multiply(A=@n12, B=@n14), B=@n14))
  n16 = FunctionInput() {
   bUsePreviewValueAsDefault=true
   description="Makes the fall phase not linear by applying a exponential curve. Values above 1 creates a easy out effect; below 1 creates a easy in effect."
   inputName=Fall Easing
   inputType=FunctionInput_Scalar
   sortPriority=9
   previewValue=<1 0 0 1>
  }
  n17 = FunctionInput() {
   bUsePreviewValueAsDefault=true
   description="Sets if the signal fall is gradual or goes back immediatelly to zero. Useful for hooking up animations that already have a full cycle that goes back to the initial position. 0 = Linear, 1 = PingPong"
   inputName=AnimationMode
   inputType=FunctionInput_Scalar
   sortPriority=99
   previewValue=<1 1 1 1>
  }
  n18 = Multiply(A=Power(Base=Saturate(None=@n15), Exp=@n16) {constExponent=1}, B=@n17)
  Add(A=@n11, B=@n18)
  n19 = FunctionInput() {
   bUsePreviewValueAsDefault=true
   description="Max value reached by the signal. Values above 1 are clamped in the final output, creating a flat signal peak. Using a Peak value that is lower than MinValue will invert the signal."
   inputName=Peak Value
   inputType=FunctionInput_Scalar
   sortPriority=4
   previewValue=<1 0 0 1>
  }
  n20 = FunctionInput() {
   bUsePreviewValueAsDefault=true
   description="Minimum value reached by the signal. Values below zero are clamped in the final output, clipping the bottom part of the signal. Using a Min value that is lower than PeakValue will invert the signal."
   inputName=Min Value
   inputType=FunctionInput_Scalar
   sortPriority=5
  }
  n21 = Multiply(A=Max(A=@n11, B=@n18), B=Subtract(A=@n19, B=@n20))
  FunctionOutput(None=Saturate(None=Add(A=@n21, B=@n20))) {outputName=Signal}
 ]
}
